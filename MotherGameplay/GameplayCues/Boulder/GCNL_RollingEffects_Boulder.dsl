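; Looping gameplay cue notify that drives the Niagara rolling effects for the Athena boulder prop.
; On looping start it caches a particle component and sets its Offset to minus half the boulder mesh's bounding sphere radius.
; Each tick it reads the physics object's movement state, setting CanSpawn to 0 after a zero-length delay when Flying, and to 1 otherwise.
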(event EventBeginPlay)

(event Collision|EventActorBeginOverlap
  (params (OtherActor :type Actor%20Object%20Reference)))

(event EventTick
  (params (DeltaSeconds :type Float%20%28single-precision%29))
  (bind (_outinstigator _outhitresult _outlinearspeed _outmovementstate) (PhysicsObject|GetMovingStateData (Variables|FortGameplayCueNotifyLoop_PhysicsObjectRolling|GetPhysicsObject)))
  (bind _fx_reference (Variables|Default|GetFX_Reference))
  (|CallParentFunction DeltaSeconds)
  (if (Utilities|Enum|Equal(Enum) _outmovementstate "Flying")
    (Utilities|FlowControl|Delay 0)
    (Niagara|SetNiagaraVariable(Int32) _fx_reference "CanSpawn" 0)
    (else
      (Niagara|SetNiagaraVariable(Int32) _fx_reference "CanSpawn" 1))))

(event EventOnLoopingStart(Niagara)
  (params (MyTarget :type Actor%20Object%20Reference) (Parameters :type Gameplay%20Cue%20Parameters%20Structure%20%28by%20ref%29) (ParticleComponents :type Array%20of%20Niagara%20Particle%20System%20Component%20Object%20References) (AudioComponents :type Array%20of%20Audio%20Component%20Object%20References))
  (Variables|Default|SetFX_Reference (Utilities|Array|Get(acopy) ParticleComponents))
  (bind _asbga_athena_boulder_prop (Utilities|Casting|CastToBGA_Athena_Boulder_Prop MyTarget))
  (Effects|Components|ParticleSystem|SetFloatParameter (Variables|Default|GetFX_Reference) "Offset" (* (.SphereRadius (Math|BoxSphereBounds|BreakBoxSphereBounds (StaticMesh|GetBounds (Variables|/Script/FortniteGame.BuildingSMActor|GetStaticMesh _asbga_athena_boulder_prop)))) -0.5)))

(fn ConstructionScript
  (|CallParentFunction))
